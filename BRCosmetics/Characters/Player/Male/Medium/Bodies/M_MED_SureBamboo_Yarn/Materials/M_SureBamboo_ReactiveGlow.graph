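# Cloth-shaded character material for the SureBamboo Yarn body, adding a reactive glow to the 2023 anime shader function chain's emissive.
# A scalar Time parameter slides two gradient windows, main and bloom, across the noise-offset FX2 mask.
# Their combined result is tinted by glow color and intensity.
# Noise comes from static and panning textures projected on pre-skinned local XZ, with UVs computed per vertex.
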Material /BRCosmetics/Characters/Player/Male/Medium/Bodies/M_MED_SureBamboo_Yarn/Materials/M_SureBamboo_ReactiveGlow
graph {
 properties {
  bAutomaticallySetUsageInEditor=false
  bUseMaterialAttributes=true
  bUsedWithMeshDeformer=true
  bUsedWithMorphTargets=true
  bUsedWithSkeletalMesh=true
  bUsedWithStaticLighting=true
  pixelDepthOffsetMode=PDOM_Legacy
  shadingModel=MSM_Cloth
  usedShadingModels=Cloth
 }
 nodes [
  ComponentMask(None=TextureCoordinate()) {r=true}
  Frac(None=Divide(A=Time()) {constB=-3})
  n1 = MaterialFunctionCall!(CustomUV) {materialFunction=/Game/Global/Materials/Character_Parent_2023/Anime_Parent_2023/MaterialFunctions/MFC_AnimeShader_2023_Pre}
  n2 = MaterialFunctionCall!(CustomLightVector, OverrideNormal_WS, STT=@n1.STT, Normal_Map=@n1.STT, DFL=@n1.STT, FX=@n1.STT) {materialFunction=/Game/Global/Materials/Character_Parent_2023/Anime_Parent_2023/MaterialFunctions/MFC_AnimeShader_2023_Calculations}
  n3 = OneMinus(None=Divide(A=ScalarParameter() {defaultValue=1, group=03_ReactiveGlow, parameterName=Time, sliderMax=1, sliderMin=-1}))
  n4 = Saturate(None=Add(A=ScalarParameter() {defaultValue=-0.5, group=03_ReactiveGlow, parameterName=Gradient_Offset, sliderMax=1, sliderMin=-1}, B=@n3))
  n5 = Add(A=@n4, B=ScalarParameter() {defaultValue=1, group=03_ReactiveGlow, parameterName=Gradient_Width, sliderMax=1})
  n6 = TextureSampleParameter2D() {
   group=03_ReactiveGlow
   parameterName=FX2
   samplerType=SAMPLERTYPE_Alpha
   texture=/BRCosmetics/Characters/Player/Male/Medium/Bodies/M_MED_SureBamboo_Yarn/Textures/T_SureBamboo_Yarn_Body_FX2
  }
  n7 = Add(A=@n6.R, B=NamedRerouteUsage())
  n8 = Multiply(
   A=Power(Base=InverseLinearInterpolate(A=@n4, B=@n5, Value=@n7)) {constExponent=2.2}
   B=ScalarParameter() {group=03_ReactiveGlow, parameterName=Gradient_Intensity, sliderMax=1}
  )
  n9 = NamedRerouteUsage()
  n10 = Saturate(None=Add(A=@n3, B=ScalarParameter() {group=03_ReactiveGlow, parameterName=GradientBloom_Offset, sliderMax=1, sliderMin=-1}))
  n11 = InverseLinearInterpolate(
   A=@n10
   B=Add(A=@n10, B=ScalarParameter() {defaultValue=0.1, group=03_ReactiveGlow, parameterName=GradientBloom_Width, sliderMax=1})
   Value=@n7
  )
  n12 = Saturate(None=Add(A=Multiply(A=@n8, B=@n9), B=Power(Base=@n11) {constExponent=2.2}))
  n13 = Multiply(
   A=VectorParameter() {group=03_ReactiveGlow, parameterName=ReactiveGlow_Color, defaultValue=<rgba 0 1 0.8168 1>}.RGB
   B=ScalarParameter() {defaultValue=5, group=03_ReactiveGlow, parameterName=ReactiveGlow_Intensity}
  )
  MaterialFunctionCall!(
   Shading_Gradient=@n2.Shading_Gradient
   Threshold_Gradient=@n2.Shading_Gradient
   Specular_Gradient=@n2.Shading_Gradient
   Sheen_Mask=@n2.Shading_Gradient
   Thickness=@n2.Shading_Gradient
   DFL_Mask=@n2.Shading_Gradient
   FX=@n2.Shading_Gradient
   Metallic_Gradient=@n2.Shading_Gradient
   RimLight_Gradient=@n2.Shading_Gradient
   NormalBuffer=@n2.Shading_Gradient
   --------
   CINE_LightSources=@n2.Shading_Gradient
   CINE_LightColor=@n2.Shading_Gradient
   ---------
   Color_Lit=@n1.Color_Lit
   Color_Shaded=@n1.Color_Lit
   SRM=@n1.Color_Lit
   Emissive=Add(A=@n1.Emissive, B=Multiply(A=@n12, B=@n13))
   DFL_Color=@n1.Color_Lit
   Specular_Color=@n1.Color_Lit
   RimLight_Color=@n1.Color_Lit
  ) {materialFunction=/Game/Global/Materials/Character_Parent_2023/Anime_Parent_2023/MaterialFunctions/MFC_AnimeShader_2023_Post}
  n14 = Multiply(
   A=ComponentMask(None=LocalPosition() {localOrigin=InstancePreSkinning}.XYZ) {b=true, r=true}
   B=ScalarParameter() {defaultValue=0.5, group=03_ReactiveGlow, parameterName=VariationNoise_Scale, sliderMax=2}
  )
  n15 = Divide(A=@n14) {constB=15}
  n16 = TextureSample(UVs=VertexInterpolator(VS=Panner(Coordinate=@n15) {speedX=0.15, speedY=-0.1}).PS) {samplerType=SAMPLERTYPE_Grayscale, texture=&LowResBlurredNoise}
  n17 = Divide(A=@n14) {constB=20}
  n18 = TextureSample(UVs=VertexInterpolator(VS=Panner(Coordinate=@n17) {speedX=-0.15, speedY=-0.05}).PS) {samplerType=SAMPLERTYPE_Grayscale, texture=&LowResBlurredNoise}
  NamedRerouteDeclaration(None=Multiply(A=Add(A=@n16.R, B=@n18.R)) {constB=0.5}) {name=Noise, nodeColor=<rgba 0 0.1562 1 1>}
  n19 = TextureSample(UVs=VertexInterpolator(VS=Divide(A=@n14) {constB=15}).PS) {samplerType=SAMPLERTYPE_Grayscale, texture=&LowResBlurredNoise}
  NamedRerouteDeclaration(None=Multiply(
    A=Add(A=@n19.R) {constB=-0.5}
    B=ScalarParameter() {defaultValue=0.1, group=03_ReactiveGlow, parameterName=VariationNoise_Strenght, sliderMax=0.2}
   ) {constB=0.02}) {name=VariationNoise, nodeColor=<rgba 0 0.1328 1 1>}
 ]
}
